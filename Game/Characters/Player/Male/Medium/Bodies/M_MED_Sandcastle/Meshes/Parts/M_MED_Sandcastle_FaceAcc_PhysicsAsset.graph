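# Physics asset for the Sandcastle male player's FaceAcc mesh, simulating front, long and back hair.
# Capsule chains use tight swing and twist limits, and most hang from tiny MSTR anchor bodies joined to the head with free rotation.
# Head, neck, upper spine and upper arm shapes provide body volumes for the hair.
# Free-rotation cross constraints tie bottom front strands to the long side chains, whose 02 bodies lack constraints to their MSTR anchors.
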PhysicsAsset /Game/Characters/Player/Male/Medium/Bodies/M_MED_Sandcastle/Meshes/Parts/M_MED_Sandcastle_FaceAcc_PhysicsAsset
bodies {
 dyn_hairfront_01_l [{shapeType=Capsule, length=4.3104, radius=1.4044, shapeName=dyn_hairfront_01_l_capsule, center=<3 0 0>, rotation=<rot 90 0 0>}]
 dyn_hairfront_01_l_MSTR [{shapeType=Capsule, length=0.01, radius=0.01}]
 dyn_hairfront_01_r [{shapeType=Capsule, length=5.3476, radius=1.3494, shapeName=dyn_hairfront_01_r_capsule, center=<2.9269 0 -0.0001>, rotation=<rot 90 0 0>}]
 dyn_hairfront_01_r_MSTR [{shapeType=Capsule, length=0.01, radius=0.01}]
 dyn_hairfront_02_l [{shapeType=Capsule, length=4.3104, radius=1.4044, shapeName=dyn_hairfront_02_l_capsule, center=<3 0 0>, rotation=<rot 90 0 0>}]
 dyn_hairfront_02_r [{shapeType=Capsule, length=5.3476, radius=1.3494, shapeName=dyn_hairfront_02_r_capsule, center=<2.9269 0 -0.0001>, rotation=<rot 90 0 0>}]
 dyn_hairfront_03_l [{shapeType=Capsule, length=4.3104, radius=1.4044, shapeName=dyn_hairfront_03_l_capsule, center=<3 0 0>, rotation=<rot 90 0 0>}]
 dyn_hairfront_03_r [{shapeType=Capsule, length=5.3476, radius=1.3494, shapeName=dyn_hairfront_03_r_capsule, center=<2.9269 0 -0.0001>, rotation=<rot 90 0 0>}]
 dyn_hairfront_bt_01_l [{shapeType=Capsule, length=5, radius=1.5, shapeName=dyn_hairfront_bt_01_l_capsule, center=<3 0 0>, rotation=<rot 90 0 0>}]
 dyn_hairfront_bt_01_l_MSTR [{shapeType=Capsule, length=0.01, radius=0.01}]
 dyn_hairfront_bt_01_r [{shapeType=Capsule, length=4.8919, radius=2, shapeName=dyn_hairfront_bt_01_r_capsule, center=<3 0 0>, rotation=<rot 90 0 0>}]
 dyn_hairfront_bt_01_r_MSTR [{shapeType=Capsule, length=0.01, radius=0.01}]
 dyn_hairfront_bt_02_l [{shapeType=Capsule, length=5, radius=2, shapeName=dyn_hairfront_bt_02_l_capsule, center=<3 0 1>, rotation=<rot 90 0 0>}]
 dyn_hairfront_bt_02_r [{shapeType=Capsule, length=4.8919, radius=2, shapeName=dyn_hairfront_bt_02_r_capsule, center=<3 0 0>, rotation=<rot 90 0 0>}]
 dyn_hairfront_bt_03_l [{shapeType=Capsule, length=5.8035, radius=1.4856, shapeName=dyn_hairfront_bt_03_l_capsule, center=<3 0 1>, rotation=<rot 90 0 0>}]
 dyn_hairfront_bt_03_r [{shapeType=Capsule, length=4.8919, radius=2, shapeName=dyn_hairfront_bt_03_r_capsule, center=<3 0 0>, rotation=<rot 90 0 0>}]
 dyn_hairfront_bt_04_l [{shapeType=Capsule, length=4.5144, radius=0.9856, shapeName=dyn_hairfront_bt_04_l_capsule, center=<3 0 1.5>, rotation=<rot 90 0 0>}]
 dyn_hairfront_bt_04_r [{shapeType=Capsule, length=6, radius=1.25, shapeName=dyn_hairfront_bt_04_r_capsule, center=<2.5597 0 0>, rotation=<rot 84.9975 180 179.9999>}]
 dyn_hairfront_bt_05_l [{shapeType=Capsule, length=3.5192, radius=0.9856, shapeName=dyn_hairfront_bt_05_l_capsule, center=<3 0 1.5>, rotation=<rot 90 0 0>}]
 dyn_hairfront_bt_05_r [{shapeType=Capsule, length=6, radius=1.25, shapeName=dyn_hairfront_bt_05_r_capsule, center=<2.5597 0 0>, rotation=<rot 84.9975 180 179.9999>}]
 dyn_hairfront_bt_06_l [{shapeType=Capsule, length=3.4181, radius=0.9856, shapeName=dyn_hairfront_bt_06_l_capsule, center=<3 0 1.5>, rotation=<rot 90 0 0>}]
 dyn_hairfront_tp_01_l [{shapeType=Capsule, length=4.8231, radius=1.0827, shapeName=dyn_hairfront_tp_01_l_capsule, center=<3 0 0>, rotation=<rot 90 0 0>}]
 dyn_hairfront_tp_01_l_MSTR [{shapeType=Capsule, length=0.01, radius=0.01}]
 dyn_hairfront_tp_01_r [{shapeType=Capsule, length=4.1671, radius=1.2022, shapeName=dyn_hairfront_tp_01_r_capsule, center=<3 0 0>, rotation=<rot 90 0 0>}]
 dyn_hairfront_tp_01_r_MSTR [{shapeType=Capsule, length=0.01, radius=0.01}]
 dyn_hairfront_tp_02_l [{shapeType=Capsule, length=4.8231, radius=1.0827, shapeName=dyn_hairfront_tp_02_l_capsule, center=<3 0 0>, rotation=<rot 90 0 0>}]
 dyn_hairfront_tp_02_r [{shapeType=Capsule, length=4.1671, radius=1.2022, shapeName=dyn_hairfront_tp_02_r_capsule, center=<3 0 0>, rotation=<rot 90 0 0>}]
 dyn_hairlong_01_l_MSTR [{shapeType=Capsule, length=0.01, radius=0.01}]
 dyn_hairlong_01_r_MSTR [{shapeType=Capsule, length=0.01, radius=0.01}]
 dyn_hairlong_02_l [{shapeType=Capsule, length=5.0251, radius=2.3907, center=<3.1927 0 0>, rotation=<rot -90 0.0197 -0.0197>}]
 dyn_hairlong_02_r [{shapeType=Capsule, length=5.0251, radius=2.3907, center=<3.1927 0 0>, rotation=<rot -90 0.0197 -0.0197>}]
 dyn_hairlong_03_l [{shapeType=Capsule, length=5.7554, radius=1.6604, center=<3.1927 0 0>, rotation=<rot -90 0.0197 -0.0197>}]
 dyn_hairlong_03_r [{shapeType=Capsule, length=5.0251, radius=2.3907, center=<3 0 0>, rotation=<rot -90 0 0>}]
 dyn_hairlong_04_l [{shapeType=Capsule, length=5.7253, radius=1.6905, center=<3.1927 0 0>, rotation=<rot -90 0.0197 -0.0197>}]
 dyn_hairlong_04_r [{shapeType=Capsule, length=5.0251, radius=2.3907, center=<3 0 0>, rotation=<rot -90 0 0>}]
 dyn_hairlong_bk_01 [{shapeType=Capsule, length=5.5251, radius=1.8907, center=<3.1927 0 0>, rotation=<rot -90 0.0197 -0.0197>}]
 dyn_hairlong_bk_01_MSTR [{shapeType=Capsule, length=0.01, radius=0.01}]
 dyn_hairlong_bk_01_l [{shapeType=Capsule, length=5.5251, radius=1.8907, center=<3.1927 0 0>, rotation=<rot -90 0.0197 -0.0197>}]
 dyn_hairlong_bk_01_l_MSTR [{shapeType=Capsule, length=0.01, radius=0.01}]
 dyn_hairlong_bk_01_r [{shapeType=Capsule, length=5.5251, radius=1.8907, center=<3.1927 0 0>, rotation=<rot -90 0.0197 -0.0197>}]
 dyn_hairlong_bk_01_r_MSTR [{shapeType=Capsule, length=0.01, radius=0.01}]
 dyn_hairlong_bk_02 [{shapeType=Capsule, length=5.5251, radius=1.8907, center=<3 0 0>, rotation=<rot -90 0 0>}]
 dyn_hairlong_bk_02_l [{shapeType=Capsule, length=5.5251, radius=1.8907, center=<3.1927 0 0>, rotation=<rot -90 0.0197 -0.0197>}]
 dyn_hairlong_bk_02_r [{shapeType=Capsule, length=5.5251, radius=1.8907, center=<3.1927 0 0>, rotation=<rot -90 0.0197 -0.0197>}]
 dyn_hairlong_bk_03 [{shapeType=Capsule, length=7.3354, radius=2.8603, center=<3 0 1.5>, rotation=<rot -90 0 0>}]
 dyn_hairlong_bk_03_l [{shapeType=Capsule, length=8.4519, radius=2.548, center=<3 0 0>, rotation=<rot -90 0 0>}]
 dyn_hairlong_bk_03_r [{shapeType=Capsule, length=8, radius=3, center=<3 0 0>, rotation=<rot -90 0 0>}]
 dyn_hairlong_bk_04 [{shapeType=Capsule, length=7.2739, radius=2.9696, center=<3 0 1.5>, rotation=<rot -90 0 0>}]
 dyn_hairlong_bk_04_l [{shapeType=Capsule, length=6, radius=2, center=<3 0 0>, rotation=<rot -90 0 0>}]
 dyn_hairlong_bk_04_r [{shapeType=Capsule, length=6, radius=2, center=<3 0 0>, rotation=<rot -90 0 0>}]
 dyn_hairlong_bk_05 [{shapeType=Capsule, length=5.7532, radius=3.0068, center=<3 0 1.5>, rotation=<rot -90 0 0>}]
 dyn_hairlong_bk_05_l [{shapeType=Capsule, length=4.409, radius=3.0068, center=<3 0 0>, rotation=<rot -90 0 0>}]
 dyn_hairlong_bk_05_r [{shapeType=Capsule, length=4.409, radius=3.0068, center=<3 0 0>, rotation=<rot -90 0 0>}]
 head [
  {shapeType=Sphere, radius=7.0336, shapeName=head_sphere, center=<-4.7045 4.6003 0>}
  {shapeType=Sphere, radius=7.0992, shapeName=head_sphere, center=<0 -1.9748 0>}
  {shapeType=Capsule, length=12.5438, radius=7.4116, center=<2.5011 1.1113 0>, rotation=<rot 90 -180 -180>}
 ]
 neck_01 [{shapeType=Capsule, length=12.3546, radius=8.7026, center=<1.796 -0.3404 0>, rotation=<rot 90 180 7.1122>}]
 spine_04_COL [{shapeType=Capsule, length=17.75, radius=9.25, center=<5.2801 3.9932 -0.0001>, rotation=<rot -0.0005 -180 -180>}]
 spine_05_COL [
  {shapeType=Capsule, length=16.6036, radius=14.4132, center=<-6.0111 -1.7451 -0.0002>, rotation=<rot 0 0 180>}
  {shapeType=Capsule, length=23.8724, radius=9.0625, center=<5.2015 -1.6347 0.0001>}
 ]
 upperarm_l [{shapeType=Capsule, length=18.5, radius=8.755, center=<8.8498 -1.3674 -2.6872>, rotation=<rot 0 -89.9995 100.0001>}]
 upperarm_r [{shapeType=Capsule, length=18.5, radius=8.755, center=<-8.8498 1.3674 2.6872>, rotation=<rot 0 -89.9995 -79.9998>}]
}
constraints [
 {bone1Name=dyn_hairlong_03_l, bone2Name=dyn_hairlong_02_l, swing1LimitDegrees=30, swing2LimitDegrees=10, twistLimitDegrees=5}
 {bone1Name=dyn_hairlong_04_l, bone2Name=dyn_hairlong_03_l, swing1LimitDegrees=30, swing2LimitDegrees=10, twistLimitDegrees=5}
 {bone1Name=dyn_hairlong_bk_01_l, bone2Name=dyn_hairlong_bk_01_l_MSTR, swing1LimitDegrees=5, swing2LimitDegrees=15, twistLimitDegrees=5}
 {bone1Name=dyn_hairlong_bk_02_l, bone2Name=dyn_hairlong_bk_01_l, swing1LimitDegrees=5, swing2LimitDegrees=30, twistLimitDegrees=2}
 {bone1Name=dyn_hairlong_bk_03_l, bone2Name=dyn_hairlong_bk_02_l, swing1LimitDegrees=5, swing2LimitDegrees=30, twistLimitDegrees=2}
 {bone1Name=dyn_hairlong_bk_04_l, bone2Name=dyn_hairlong_bk_03_l, swing1LimitDegrees=5, swing2LimitDegrees=30, twistLimitDegrees=2}
 {bone1Name=dyn_hairlong_bk_05_l, bone2Name=dyn_hairlong_bk_04_l, swing1LimitDegrees=5, swing2LimitDegrees=30, twistLimitDegrees=2}
 {bone1Name=dyn_hairlong_bk_01, bone2Name=dyn_hairlong_bk_01_MSTR, swing1LimitDegrees=5, swing2LimitDegrees=45, twistLimitDegrees=5}
 {bone1Name=dyn_hairlong_bk_02, bone2Name=dyn_hairlong_bk_01, swing1LimitDegrees=10, swing2LimitDegrees=30, twistLimitDegrees=2}
 {bone1Name=dyn_hairlong_bk_03, bone2Name=dyn_hairlong_bk_02, swing1LimitDegrees=10, swing2LimitDegrees=30, twistLimitDegrees=2}
 {bone1Name=dyn_hairlong_bk_04, bone2Name=dyn_hairlong_bk_03, swing1LimitDegrees=10, swing2LimitDegrees=30, twistLimitDegrees=2}
 {bone1Name=dyn_hairlong_bk_05, bone2Name=dyn_hairlong_bk_04, swing1LimitDegrees=10, swing2LimitDegrees=30, twistLimitDegrees=2}
 {bone1Name=dyn_hairfront_tp_01_l, bone2Name=dyn_hairfront_tp_01_l_MSTR, swing1LimitDegrees=10, swing2LimitDegrees=10, twistLimitDegrees=2}
 {bone1Name=dyn_hairfront_tp_02_l, bone2Name=dyn_hairfront_tp_01_l, swing1LimitDegrees=10, swing2LimitDegrees=10, twistLimitDegrees=2}
 {bone1Name=dyn_hairfront_01_l, bone2Name=dyn_hairfront_01_l_MSTR, swing1LimitDegrees=5, swing2LimitDegrees=0, twistLimitDegrees=2}
 {bone1Name=dyn_hairfront_02_l, bone2Name=dyn_hairfront_01_l, swing1LimitDegrees=10, swing2LimitDegrees=10, twistLimitDegrees=2}
 {bone1Name=dyn_hairfront_03_l, bone2Name=dyn_hairfront_02_l, swing1LimitDegrees=10, swing2LimitDegrees=10, twistLimitDegrees=2}
 {bone1Name=dyn_hairlong_bk_01_r, bone2Name=dyn_hairlong_bk_01_r_MSTR, swing1LimitDegrees=5, swing2LimitDegrees=45, twistLimitDegrees=5}
 {bone1Name=dyn_hairlong_bk_02_r, bone2Name=dyn_hairlong_bk_01_r, swing1LimitDegrees=5, swing2LimitDegrees=30, twistLimitDegrees=2}
 {bone1Name=dyn_hairlong_bk_03_r, bone2Name=dyn_hairlong_bk_02_r, swing1LimitDegrees=5, swing2LimitDegrees=30, twistLimitDegrees=2}
 {bone1Name=dyn_hairlong_bk_04_r, bone2Name=dyn_hairlong_bk_03_r, swing1LimitDegrees=5, swing2LimitDegrees=30, twistLimitDegrees=2}
 {bone1Name=dyn_hairlong_bk_05_r, bone2Name=dyn_hairlong_bk_04_r, swing1LimitDegrees=5, swing2LimitDegrees=30, twistLimitDegrees=2}
 {bone1Name=dyn_hairlong_03_r, bone2Name=dyn_hairlong_02_r, swing1LimitDegrees=30, swing2LimitDegrees=10, twistLimitDegrees=5}
 {bone1Name=dyn_hairlong_04_r, bone2Name=dyn_hairlong_03_r, swing1LimitDegrees=30, swing2LimitDegrees=10, twistLimitDegrees=5}
 {bone1Name=dyn_hairfront_01_r, bone2Name=dyn_hairfront_01_r_MSTR, swing1LimitDegrees=20, swing2LimitDegrees=20, twistLimitDegrees=2}
 {bone1Name=dyn_hairfront_02_r, bone2Name=dyn_hairfront_01_r, swing1LimitDegrees=10, swing2LimitDegrees=10, twistLimitDegrees=2}
 {bone1Name=dyn_hairfront_03_r, bone2Name=dyn_hairfront_02_r, swing1LimitDegrees=5, swing2LimitDegrees=5, twistLimitDegrees=2}
 {bone1Name=dyn_hairfront_tp_01_r, bone2Name=dyn_hairfront_tp_01_r_MSTR, swing1LimitDegrees=10, swing2LimitDegrees=10, twistLimitDegrees=2}
 {bone1Name=dyn_hairfront_tp_02_r, bone2Name=dyn_hairfront_tp_01_r, swing1LimitDegrees=10, swing2LimitDegrees=10, twistLimitDegrees=2}
 {bone1Name=dyn_hairfront_bt_01_r, bone2Name=dyn_hairfront_bt_01_r_MSTR, swing1LimitDegrees=15, swing2LimitDegrees=15, twistLimitDegrees=5}
 {bone1Name=dyn_hairfront_bt_02_r, bone2Name=dyn_hairfront_bt_01_r, swing1LimitDegrees=15, swing2LimitDegrees=15, twistLimitDegrees=5}
 {bone1Name=dyn_hairfront_bt_03_r, bone2Name=dyn_hairfront_bt_02_r, swing1LimitDegrees=15, swing2LimitDegrees=15, twistLimitDegrees=5}
 {bone1Name=dyn_hairfront_bt_04_r, bone2Name=dyn_hairfront_bt_03_r, swing1LimitDegrees=45, swing2LimitDegrees=45, twistLimitDegrees=5}
 {bone1Name=dyn_hairfront_bt_05_r, bone2Name=dyn_hairfront_bt_04_r, swing1LimitDegrees=45, swing2LimitDegrees=45, twistLimitDegrees=5}
 {bone1Name=dyn_hairfront_bt_01_l, bone2Name=dyn_hairfront_bt_01_l_MSTR, swing1LimitDegrees=15, swing2LimitDegrees=15, twistLimitDegrees=2}
 {bone1Name=dyn_hairfront_bt_02_l, bone2Name=dyn_hairfront_bt_01_l, swing1LimitDegrees=15, swing2LimitDegrees=15, twistLimitDegrees=2}
 {bone1Name=dyn_hairfront_bt_03_l, bone2Name=dyn_hairfront_bt_02_l, swing1LimitDegrees=10, swing2LimitDegrees=10, twistLimitDegrees=2}
 {bone1Name=dyn_hairfront_bt_04_l, bone2Name=dyn_hairfront_bt_03_l, swing1LimitDegrees=10, swing2LimitDegrees=10, twistLimitDegrees=2}
 {bone1Name=dyn_hairfront_bt_05_l, bone2Name=dyn_hairfront_bt_04_l, swing1LimitDegrees=10, swing2LimitDegrees=10, twistLimitDegrees=2}
 {bone1Name=dyn_hairfront_bt_06_l, bone2Name=dyn_hairfront_bt_04_l, swing1LimitDegrees=10, swing2LimitDegrees=10, twistLimitDegrees=2}
 {
  bone1Name=dyn_hairlong_bk_01_l_MSTR
  bone2Name=head
  swing1LimitDegrees=45
  swing1Motion=Free
  swing2LimitDegrees=45
  swing2Motion=Free
  twistLimitDegrees=45
  twistMotion=Free
 }
 {
  bone1Name=dyn_hairlong_bk_01_MSTR
  bone2Name=head
  swing1LimitDegrees=45
  swing1Motion=Free
  swing2LimitDegrees=45
  swing2Motion=Free
  twistLimitDegrees=45
  twistMotion=Free
 }
 {
  bone1Name=dyn_hairfront_tp_01_l_MSTR
  bone2Name=head
  swing1LimitDegrees=45
  swing1Motion=Free
  swing2LimitDegrees=45
  swing2Motion=Free
  twistLimitDegrees=45
  twistMotion=Free
 }
 {
  bone1Name=dyn_hairfront_01_l_MSTR
  bone2Name=head
  swing1LimitDegrees=45
  swing1Motion=Free
  swing2LimitDegrees=45
  swing2Motion=Free
  twistLimitDegrees=45
  twistMotion=Free
 }
 {
  bone1Name=dyn_hairlong_bk_01_r_MSTR
  bone2Name=head
  swing1LimitDegrees=45
  swing1Motion=Free
  swing2LimitDegrees=45
  swing2Motion=Free
  twistLimitDegrees=45
  twistMotion=Free
 }
 {
  bone1Name=dyn_hairfront_01_r_MSTR
  bone2Name=head
  swing1LimitDegrees=45
  swing1Motion=Free
  swing2LimitDegrees=45
  swing2Motion=Free
  twistLimitDegrees=45
  twistMotion=Free
 }
 {
  bone1Name=dyn_hairfront_tp_01_r_MSTR
  bone2Name=head
  swing1LimitDegrees=45
  swing1Motion=Free
  swing2LimitDegrees=45
  swing2Motion=Free
  twistLimitDegrees=45
  twistMotion=Free
 }
 {
  bone1Name=dyn_hairfront_bt_01_r_MSTR
  bone2Name=head
  swing1LimitDegrees=45
  swing1Motion=Free
  swing2LimitDegrees=45
  swing2Motion=Free
  twistLimitDegrees=45
  twistMotion=Free
 }
 {
  bone1Name=dyn_hairfront_bt_01_l_MSTR
  bone2Name=head
  swing1LimitDegrees=45
  swing1Motion=Free
  swing2LimitDegrees=45
  swing2Motion=Free
  twistLimitDegrees=45
  twistMotion=Free
 }
 {
  bone1Name=head
  bone2Name=neck_01
  swing1LimitDegrees=45
  swing1Motion=Free
  swing2LimitDegrees=45
  swing2Motion=Free
  twistLimitDegrees=45
  twistMotion=Free
 }
 {
  bone1Name=dyn_hairlong_01_l_MSTR
  bone2Name=head
  swing1LimitDegrees=45
  swing1Motion=Free
  swing2LimitDegrees=45
  swing2Motion=Free
  twistLimitDegrees=45
  twistMotion=Free
 }
 {
  bone1Name=dyn_hairlong_01_r_MSTR
  bone2Name=head
  swing1LimitDegrees=45
  swing1Motion=Free
  swing2LimitDegrees=45
  swing2Motion=Free
  twistLimitDegrees=45
  twistMotion=Free
 }
 {
  bone1Name=dyn_hairfront_bt_02_l
  bone2Name=dyn_hairlong_03_l
  swing1LimitDegrees=45
  swing1Motion=Free
  swing2LimitDegrees=45
  swing2Motion=Free
  twistLimitDegrees=45
  twistMotion=Free
 }
 {
  bone1Name=dyn_hairfront_bt_02_r
  bone2Name=dyn_hairlong_03_r
  swing1LimitDegrees=45
  swing1Motion=Free
  swing2LimitDegrees=45
  swing2Motion=Free
  twistLimitDegrees=45
  twistMotion=Free
 }
 {
  bone1Name=dyn_hairfront_bt_03_r
  bone2Name=dyn_hairlong_04_r
  swing1LimitDegrees=45
  swing1Motion=Free
  swing2LimitDegrees=45
  swing2Motion=Free
  twistLimitDegrees=45
  twistMotion=Free
 }
]
