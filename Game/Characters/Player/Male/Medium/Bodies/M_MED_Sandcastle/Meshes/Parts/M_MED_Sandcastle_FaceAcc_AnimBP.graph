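# Post-process style AnimBP for the Sandcastle male skin's face accessory part, posing and simulating its hair bones on top of the pose copied from the parent mesh.
# Look-At, constraint and additive bone offsets feed a spine_05-based RigidBody sim, followed by further corrective offsets.
# The EventGraph drives corrective alphas from head-to-spine rotation, targeting and the front-end idle montage.
# Skydiving force replaces sim gravity, and dynamics reset when skydiving ends.
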AnimBlueprint /Game/Characters/Player/Male/Medium/Bodies/M_MED_Sandcastle/Meshes/Parts/M_MED_Sandcastle_FaceAcc_AnimBP
target_skeleton=&/M_MED_Sandcastle_FaceAcc_Skeleton
graphs [
 AnimGraph {nodes [
   n1 = Get headLookUpAlpha() -> (headLookUpAlpha:double)
   n2 = Look At - Bone: dyn_hairlong_bk_01_l_MSTR!(ComponentPose=Local To Component(LocalPose=Copy Pose From Mesh() {bUseAttachedParent=true}), Alpha=@n1, LookAtLocation:Vector="0,0,0") -> (Pose) {
    alphaBoolBlend {blendOption=Linear}
    boneToModify {boneName=dyn_hairlong_bk_01_l_MSTR}
    lookAtLocation {x=0, z=10}
    lookAtTarget {boneReference {boneName=dyn_hairlong_bk_01_l_aim}}
    lookAt_Axis {axis=<1 0>}
    lookUp_Axis {axis {x=0, z=1}}
   }
   n3 = Look At - Bone: dyn_hairlong_bk_01_MSTR!(ComponentPose=@n2, Alpha=@n1, LookAtLocation:Vector="0,0,0") -> (Pose) {
    alphaBoolBlend {blendOption=Linear}
    boneToModify {boneName=dyn_hairlong_bk_01_MSTR}
    lookAtLocation {x=0, z=10}
    lookAtTarget {boneReference {boneName=dyn_hairlong_bk_01_aim}}
    lookAt_Axis {axis=<1 0>}
    lookUp_Axis {axis {x=0, z=1}}
   }
   n4 = Look At - Bone: dyn_hairlong_bk_01_r_MSTR!(ComponentPose=@n3, Alpha=@n1, LookAtLocation:Vector="0,0,0") -> (Pose) {
    alphaBoolBlend {blendOption=Linear}
    boneToModify {boneName=dyn_hairlong_bk_01_r_MSTR}
    lookAtLocation {x=0, z=10}
    lookAtTarget {boneReference {boneName=dyn_hairlong_bk_01_r_aim}}
    lookAt_Axis {axis=<1 0>}
    lookUp_Axis {axis {x=0, z=1}}
   }
   n5 = "Constraint - dyn_hairlong_05_r (P)"(ComponentPose=@n4, Alpha=@n1, ConstraintWeights_0=1) {
    alphaBoolBlend {blendOption=Linear}
    boneToModify {boneName=dyn_hairlong_05_r}
    constraintSetup [{targetBone {boneName=dyn_hairfront_bt_02_r}}]
    constraintWeights [1]
   }
   n6 = "Transform (Modify) Bone - Bone: dyn_hairlong_05_r"(ComponentPose=@n5, Alpha=1, Translation="-3.239,0.000,-0.000", Rotation="0,0,0", Scale="1,1,1") {
    rotationMode=BMM_Additive
    rotationSpace=BCS_BoneSpace
    translationMode=BMM_Additive
    translationSpace=BCS_BoneSpace
    alphaBoolBlend {blendOption=Linear}
    boneToModify {boneName=dyn_hairlong_05_r}
    translation {x=-3.239}
   }
   n7 = "Transform (Modify) Bone - Bone: dyn_hairfront_bt_01_l_MSTR"(
    ComponentPose=@n6
    Alpha=Get Is FrontEnd look left() -> (Is FrontEnd look left:double)
    Translation="0,0,0"
    Rotation="-0.708,-13.497,-6.314"
    Scale="1,1,1"
   ) {
    rotationMode=BMM_Additive
    rotationSpace=BCS_BoneSpace
    translationMode=BMM_Additive
    translationSpace=BCS_BoneSpace
    alphaBoolBlend {blendOption=Linear}
    boneToModify {boneName=dyn_hairfront_bt_01_l_MSTR}
    rotation=<rot -0.708 -13.497 -6.314>
   }
   n8 = "Transform (Modify) Bone - Bone: dyn_hairlong_bk_06_r"(
    ComponentPose=@n7
    Alpha=Get isTargetingAlpha() -> (isTargetingAlpha:double)
    Translation="-1.827,-1.755,2.216"
    Rotation="0,0,0"
    Scale="1,1,1"
   ) {
    rotationMode=BMM_Additive
    rotationSpace=BCS_BoneSpace
    translationMode=BMM_Additive
    translationSpace=BCS_BoneSpace
    alphaBoolBlend {blendOption=Linear}
    alphaScaleBiasClamp {bInterpResult=true}
    boneToModify {boneName=dyn_hairlong_bk_06_r}
    translation=<-1.827 -1.755 2.216>
   }
   n9 = Get isFrontEndStorePoseAlpha() -> (isFrontEndStorePoseAlpha:double)
   n10 = "Transform (Modify) Bone - Bone: dyn_hairfront_bt_01_l_MSTR"(ComponentPose=@n8, Alpha=@n9, Translation="0,0,0", Rotation="25.000,0.000,0.000", Scale="1,1,1") {
    rotationMode=BMM_Additive
    rotationSpace=BCS_BoneSpace
    scaleMode=BMM_Replace
    scaleSpace=BCS_BoneSpace
    translationMode=BMM_Additive
    translationSpace=BCS_BoneSpace
    alphaBoolBlend {blendOption=Linear}
    boneToModify {boneName=dyn_hairfront_bt_01_l_MSTR}
    rotation {pitch=25}
   }
   n11 = "Transform (Modify) Bone - Bone: spine_05_COL"(ComponentPose=@n10, Alpha=@n9, Translation="-2,0,-2", Rotation="0,0,0", Scale="1,1,1") {
    rotationMode=BMM_Additive
    rotationSpace=BCS_BoneSpace
    scaleMode=BMM_Replace
    scaleSpace=BCS_BoneSpace
    translationMode=BMM_Additive
    translationSpace=BCS_BoneSpace
    alphaBoolBlend {blendOption=Linear}
    boneToModify {boneName=spine_05_COL}
    translation {x=-2, z=-2}
   }
   n12 = RigidBody!(ComponentPose=@n11, Alpha, OverrideWorldGravity:Vector=Get windVector() -> (windVector:Vector), ExternalForce) -> (Pose) {
    bClampLinearTranslationLimitToRefPose=true
    bOverrideWorldGravity=true
    overridePhysicsAsset=&/M_MED_Sandcastle_FaceAcc_PhysicsAsset
    simulationSpace=BaseBoneSpace
    alphaBoolBlend {blendOption=Linear}
    baseBoneRef {boneName=spine_05}
    componentAppliedLinearAccClamp=<3000 3000 3000>
    componentLinearAccScale=<0.3 0.3 0.3>
    componentLinearVelScale=<0.3 0.3 0.35>
    overrideWorldGravity {z=-980}
   }
   n13 = Get isFrontEndStorePoseAlpha()
   n14 = "Transform (Modify) Bone - Bone: dyn_hairfront_bt_03_l"(ComponentPose=@n12, Alpha=@n13, Translation="0,0,0", Rotation="-4.946,-9.888,3.394", Scale="1,1,1") {
    rotationMode=BMM_Additive
    rotationSpace=BCS_BoneSpace
    scaleMode=BMM_Replace
    scaleSpace=BCS_BoneSpace
    translationMode=BMM_Additive
    translationSpace=BCS_BoneSpace
    alphaBoolBlend {blendOption=Linear}
    boneToModify {boneName=dyn_hairfront_bt_03_l}
    rotation=<rot -4.946 -9.888 3.394>
   }
   n15 = "Transform (Modify) Bone - Bone: dyn_hairfront_bt_04_l"(ComponentPose=@n14, Alpha=@n13, Translation="0,0,0", Rotation="15.000,-0.003,0.000", Scale="1,1,1") {
    rotationMode=BMM_Additive
    rotationSpace=BCS_BoneSpace
    scaleMode=BMM_Replace
    scaleSpace=BCS_BoneSpace
    translationMode=BMM_Additive
    translationSpace=BCS_BoneSpace
    alphaBoolBlend {blendOption=Linear}
    boneToModify {boneName=dyn_hairfront_bt_04_l}
    rotation {pitch=15, yaw=-0.003}
   }
   n16 = "Transform (Modify) Bone - Bone: dyn_hairfront_bt_04_r"(ComponentPose=@n15, Alpha=@n13, Translation="0,0,0", Rotation="20.000,0.001,-0.002", Scale="1,1,1") {
    rotationMode=BMM_Additive
    rotationSpace=BCS_BoneSpace
    scaleMode=BMM_Replace
    scaleSpace=BCS_BoneSpace
    translationMode=BMM_Additive
    translationSpace=BCS_BoneSpace
    alphaBoolBlend {blendOption=Linear}
    boneToModify {boneName=dyn_hairfront_bt_04_r}
    rotation=<rot 20 0.001 -0.002>
   }
   n17 = "Transform (Modify) Bone - Bone: dyn_hairlong_bk_01_r_MSTR"(ComponentPose=@n16, Alpha=@n13, Translation="0,0,0", Rotation="-3.713,9.753,-0.009", Scale="1,1,1") {
    rotationMode=BMM_Additive
    rotationSpace=BCS_BoneSpace
    scaleMode=BMM_Replace
    scaleSpace=BCS_BoneSpace
    translationMode=BMM_Additive
    translationSpace=BCS_BoneSpace
    alphaBoolBlend {blendOption=Linear}
    boneToModify {boneName=dyn_hairlong_bk_01_r_MSTR}
    rotation=<rot -3.713 9.753 -0.009>
   }
   n18 = "Transform (Modify) Bone - Bone: dyn_hairlong_bk_01_MSTR"(ComponentPose=@n17, Alpha=@n13, Translation="0,0,0", Rotation="-1.979,12.014,7.939", Scale="1,1,1") {
    rotationMode=BMM_Additive
    rotationSpace=BCS_BoneSpace
    scaleMode=BMM_Replace
    scaleSpace=BCS_BoneSpace
    translationMode=BMM_Additive
    translationSpace=BCS_BoneSpace
    alphaBoolBlend {blendOption=Linear}
    boneToModify {boneName=dyn_hairlong_bk_01_MSTR}
    rotation=<rot -1.9791 12.014 7.939>
   }
   n19 = Get headLookUpAlpha()
   n20 = "Transform (Modify) Bone - Bone: clavicle_extra_roll_01_r"(ComponentPose=@n18, Alpha=@n19, Translation="0,0,0", Rotation="0.000,0.000,-35.000", Scale="1,1,1") {
    rotationMode=BMM_Additive
    rotationSpace=BCS_BoneSpace
    translationMode=BMM_Additive
    translationSpace=BCS_BoneSpace
    alphaBoolBlend {blendOption=Linear}
    boneToModify {boneName=clavicle_extra_roll_01_r}
    rotation {roll=-35.0003}
   }
   n21 = "Transform (Modify) Bone - Bone: dyn_hairlong_bk_01_r"(ComponentPose=@n20, Alpha=@n19, Translation="0,0,0", Rotation="0.000,5,0.000", Scale="1,1,1") {
    rotationMode=BMM_Additive
    rotationSpace=BCS_BoneSpace
    translationMode=BMM_Additive
    translationSpace=BCS_BoneSpace
    alphaBoolBlend {blendOption=Linear}
    boneToModify {boneName=dyn_hairlong_bk_01_r}
    rotation {yaw=5}
   }
   n22 = "Transform (Modify) Bone - Bone: dyn_hairfront_bt_01_l_MSTR"(ComponentPose=@n21, Alpha=@n19, Translation="0,0,0", Rotation="6.908,-11.405,-0.156", Scale="1,1,1") {
    rotationMode=BMM_Additive
    rotationSpace=BCS_BoneSpace
    translationMode=BMM_Additive
    translationSpace=BCS_BoneSpace
    alphaBoolBlend {blendOption=Linear}
    boneToModify {boneName=dyn_hairfront_bt_01_l_MSTR}
    rotation=<rot 6.908 -11.405 -0.156>
   }
   n23 = Get headLookUpAlpha()
   n24 = "Transform (Modify) Bone - Bone: dyn_hairlong_bk_03"(ComponentPose=@n22, Alpha=@n23, Translation="0,0,0", Rotation="-10.000,-0.002,0.000", Scale="1,1,1") {
    rotationMode=BMM_Additive
    rotationSpace=BCS_BoneSpace
    translationMode=BMM_Additive
    translationSpace=BCS_BoneSpace
    alphaBoolBlend {blendOption=Linear}
    alphaScaleBiasClamp {interpSpeedDecreasing=5, interpSpeedIncreasing=5}
    boneToModify {boneName=dyn_hairlong_bk_03}
    rotation {pitch=-10, yaw=-0.002}
   }
   n25 = "Transform (Modify) Bone - Bone: dyn_hairlong_bk_04"(ComponentPose=@n24, Alpha=@n23, Translation="0,0,0", Rotation="-10.000,-0.002,0.000", Scale="1,1,1") {
    rotationMode=BMM_Additive
    rotationSpace=BCS_BoneSpace
    translationMode=BMM_Additive
    translationSpace=BCS_BoneSpace
    alphaBoolBlend {blendOption=Linear}
    alphaScaleBiasClamp {interpSpeedDecreasing=5, interpSpeedIncreasing=5}
    boneToModify {boneName=dyn_hairlong_bk_04}
    rotation {pitch=-10, yaw=-0.002}
   }
   n26 = "Transform (Modify) Bone - Bone: dyn_hairlong_bk_03_r"(ComponentPose=@n25, Alpha=@n23, Translation="0,0,0", Rotation="-12,-0.001,0.000", Scale="1,1,1") {
    rotationMode=BMM_Additive
    rotationSpace=BCS_BoneSpace
    translationMode=BMM_Additive
    translationSpace=BCS_BoneSpace
    alphaBoolBlend {blendOption=Linear}
    alphaScaleBiasClamp {interpSpeedDecreasing=5, interpSpeedIncreasing=5}
    boneToModify {boneName=dyn_hairlong_bk_03_r}
    rotation {pitch=-12, yaw=-0.001}
   }
   n27 = "Transform (Modify) Bone - Bone: dyn_hairlong_bk_04_r"(ComponentPose=@n26, Alpha=@n23, Translation="0,0,0", Rotation="-12,-0.005,-0.001", Scale="1,1,1") {
    rotationMode=BMM_Additive
    rotationSpace=BCS_BoneSpace
    translationMode=BMM_Additive
    translationSpace=BCS_BoneSpace
    alphaBoolBlend {blendOption=Linear}
    alphaScaleBiasClamp {interpSpeedDecreasing=5, interpSpeedIncreasing=5}
    boneToModify {boneName=dyn_hairlong_bk_04_r}
    rotation=<rot -12 -0.005 -0.001>
   }
   n28 = "Transform (Modify) Bone - Bone: dyn_hairlong_bk_03"(
    ComponentPose=@n27
    Alpha=Get headAndAimAlpha() -> (headAndAimAlpha:double)
    Translation="0,0,0"
    Rotation="-0.001,10.000,0.001"
    Scale="1,1,1"
   ) {
    rotationMode=BMM_Additive
    rotationSpace=BCS_BoneSpace
    translationMode=BMM_Additive
    translationSpace=BCS_BoneSpace
    alphaBoolBlend {blendOption=Linear}
    alphaScaleBiasClamp {bInterpResult=true, interpSpeedDecreasing=5, interpSpeedIncreasing=5}
    boneToModify {boneName=dyn_hairlong_bk_03}
    rotation=<rot -0.001 10.0003 0.001>
   }
   n29 = "Transform (Modify) Bone - Bone: dyn_hairlong_bk_01_r_MSTR"(ComponentPose=@n28, Alpha=1, Translation="0,0,0", Rotation="10,0,0", Scale="1,1,1") {
    rotationMode=BMM_Additive
    rotationSpace=BCS_BoneSpace
    scaleMode=BMM_Replace
    scaleSpace=BCS_BoneSpace
    translationMode=BMM_Additive
    translationSpace=BCS_BoneSpace
    alphaBoolBlend {blendOption=Linear}
    alphaScaleBiasClamp {bInterpResult=true, interpSpeedDecreasing=5, interpSpeedIncreasing=5}
    boneToModify {boneName=dyn_hairlong_bk_01_r_MSTR}
    rotation {pitch=10}
   }
   n30 = "Transform (Modify) Bone - Bone: dyn_hairlong_bk_01_MSTR"(ComponentPose=@n29, Alpha=1, Translation="0,0,0", Rotation="10,0,0", Scale="1,1,1") {
    rotationMode=BMM_Additive
    rotationSpace=BCS_BoneSpace
    scaleMode=BMM_Replace
    scaleSpace=BCS_BoneSpace
    translationMode=BMM_Additive
    translationSpace=BCS_BoneSpace
    alphaBoolBlend {blendOption=Linear}
    alphaScaleBiasClamp {bInterpResult=true, interpSpeedDecreasing=5, interpSpeedIncreasing=5}
    boneToModify {boneName=dyn_hairlong_bk_01_MSTR}
    rotation {pitch=10}
   }
   n31 = "Transform (Modify) Bone - Bone: dyn_hairlong_bk_01_l_MSTR"(ComponentPose=@n30, Alpha=1, Translation="0,0,0", Rotation="2,-2,0", Scale="1,1,1") {
    rotationMode=BMM_Additive
    rotationSpace=BCS_BoneSpace
    translationMode=BMM_Additive
    translationSpace=BCS_BoneSpace
    alphaBoolBlend {blendOption=Linear}
    alphaScaleBiasClamp {bInterpResult=true, interpSpeedDecreasing=5, interpSpeedIncreasing=5}
    boneToModify {boneName=dyn_hairlong_bk_01_l_MSTR}
    rotation {pitch=2, yaw=-2}
   }
   Output Pose(Result=Component To Local(ComponentPose=@n31)) {layerGroup=None}
   Get isSkydivingAlpha() -> (isSkydivingAlpha:double)
  ]}
 EventGraph {nodes [
   Event BlueprintUpdateAnimation()
   > n1 = Sequence!(execute) -> (then_0, then_1, then_2:Exec, then_3:Exec, then_4:Exec, then_5:Exec)
   n2 = Break Fort Anim Input Skydiving External Force(FortAnimInput_SkydivingExternalForce=Get SkydivingExternalForceInput()) -> (
    bUseSkydivingVectorForce=~
    SkydivingVectorForce=~
    HeadToPelvisDirection=~
    FloatingMultiplier=~
    FloatingAdditive=~
    DivingMultiplier=~
    DivingAdditive=~
    ParachutingMultiplier=~
    ParachutingAdditive=~
   )
   > Set windVector(windVector:Vector=SelectVector(A=vector * float(A=@n2.SkydivingVectorForce, B=0.15), B="0, 0,-980", bPickA=Get bIsSkyDiving())) -> (Output_Get:Vector)
   n3 = GetPartSkeletalMeshComponent()
   n4 = GetAlphaFromReferenceTransformDelta(SkeletalMeshComponent=@n3, BoneName=head, BaseName=spine_05, DeltaComponentType=QuaternionTwist, InputRange=(X=0,Y=30), TwistAxis="0,0,1")
   Set headLookUpAlpha(execute=@n1.then_1, headLookUpAlpha:double=@n4) -> (Output_Get:double)
   > n5 = Set headAndAimAlpha(headAndAimAlpha:double=SelectFloat(A=@n4, B=0, bPickA=Get bIsTargeting())) -> (Output_Get:double)
   n6 = GetDeltaTransformFromRefPose!(self=@n3, BoneName=head, BaseName=spine_05) -> (
    ReturnValue_Location:Vector
    ReturnValue_Rotation_Roll:float
    ReturnValue_Rotation_Pitch:float
    ReturnValue_Rotation_Yaw:float
    ReturnValue_Scale:Vector
   )
   n7 = GetAnimInstance(self=GetOwnerSkeletalMeshComponent())
   n8 = Set isFrontEndStorePoseAlpha(
    execute=@n9.Is Valid
    isFrontEndStorePoseAlpha:double=SelectFloat(
     A=1
     B=0
     bPickA=Montage_IsPlaying(self=@n7, Montage=/Game/Animation/Game/MainPlayer/Menu/BR/CustomSkins/_s13/SandCastle/Sandcastle_Male_Idle_Montage)
    )
   ) -> (Output_Get:double)
   Set Is FrontEnd look left(
    execute=@n5.then
    Is FrontEnd look left:double=float * float(A=MapRangeClamped(Value=@n6.ReturnValue_Rotation_Roll, InRangeA=33, InRangeB=45), B=@n8) {is=title K2Node_PromotableOperator}
   ) -> (Output_Get:double)
   Set isSkydivingAlpha(execute=@n1.then_2, isSkydivingAlpha:double=SelectFloat(A=0.5, B=1, bPickA=Get bIsSkyDiving())) -> (Output_Get:double)
   n9 = Is Valid(exec=@n1.then_3, InputObject=@n7)
   Event OnBeginTargeting() {is=title K2Node_Event}
   > Set isTargetingAlpha(isTargetingAlpha:double=1) -> (Output_Get:double)
   Event OnEndTargeting() {is=title K2Node_Event}
   > Set isTargetingAlpha(isTargetingAlpha=0)
   Event OnEndSkydiving()
   > ResetDynamics(InTeleportType=TeleportPhysics)
   > ResetDynamics(InTeleportType=ResetPhysics)
   TryGetPawnOwner()
  ]}
]
variables [
 headAndAimAlpha {type=number}
 headLookUpAlpha {type=number}
 Is FrontEnd look left {type=number}
 isFrontEndStorePoseAlpha {type=number}
 isSkydivingAlpha {type=number}
 isTargetingAlpha {type=number}
 windVector {type=Vector}
]
